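# Prototype pseudo-glass material for the AxisMandate set, using the Subsurface shading model with default opaque blending in place of translucency.
# It declares scalar parameters for Metallic, Specular, Roughness and Opacity, plus vector parameters for BaseColor and SubsurfaceColor, with no textures or math nodes.
# It is flagged for Nanite use, and automatic usage setting in the editor is off.
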Material /Hera/Sets/AxisMandate/Materials/_Proto/M_AxisMandate_PseudoGlass
graph {
 properties {bAutomaticallySetUsageInEditor=false, bUsedWithNanite=true, shadingModel=MSM_Subsurface, usedShadingModels=Subsurface}
 nodes [
  ScalarParameter() {parameterName=Metallic}
  ScalarParameter() {defaultValue=0.5, parameterName=Specular}
  ScalarParameter() {defaultValue=0.2, parameterName=Roughness}
  ScalarParameter() {defaultValue=0.5, parameterName=Opacity}
  VectorParameter() {parameterName=BaseColor, defaultValue=<rgba 0.5 0.5 0.5 1>}
  VectorParameter() {parameterName=SubsurfaceColor, defaultValue=<rgba 0.5 0.5 0.5 1>}
 ]
}
